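# Masked, two-sided, lit sprite-particle material for barrier chunk effects in Figment Season 10.
# Computes a per-particle spin about one of three randomly picked axes, with angle based on distance from world origin and offset blended in later in life.
# Particle color gets four hue shifts that fade with age, then age-driven desaturation, feeding a time-of-day color function.
# No links to material outputs appear.
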Material /Figment_S10/Athena/Effects/Materials/Barrier/M_Barrier_Base_Chunks_GPU_Pixel
graph {
 properties {
  bEnableResponsiveAA=true
  bUsedWithParticleSprites=true
  blendMode=BLEND_Masked
  opacityMaskClipValue=0.1
  pixelDepthOffsetMode=PDOM_Legacy
  translucencyLightingMode=TLM_VolumetricPerVertexDirectional
  twoSided=true
  usedShadingModels=Default Lit
 }
 nodes [
  n1 = MaterialFunctionCall!(
   Random Picker=ParticleRandom()
   Texture Input
   Option 1=VectorParameter() {parameterName=NormalizeRotationAxis_1, defaultValue=<rgba 1 0.3 -0.2 0>}.RGB
   Option 2=VectorParameter() {parameterName=NormalizeRotationAxis_2, defaultValue=<rgba -0.44 1 0.13 0>}.RGB
   Option 3=VectorParameter() {parameterName=NormalizeRotationAxis_3, defaultValue=<rgba 0.06 -0.22 1 0>}.RGB
  ) {materialFunction=/Game/Effects/Fort_Effects/Materials/_Functions/MF_RGB_Random_01}
  n2 = Divide(A=Distance(A=Constant3Vector(), B=ParticlePositionWS()), B=ScalarParameter() {defaultValue=150000, parameterName=Spin Scale})
  n3 = Clamp(None=ParticleRandom()) {maxDefault=2, minDefault=0.15}
  n4 = ParticleRelativeTime()
  LinearInterpolate(
   A=Constant()
   B=RotateAboutAxis(
    NormalizedRotationAxis=Normalize(VectorInput=@n1.Vec3 Out Only)
    RotationAngle=Multiply(A=@n2, B=@n3)
    PivotPoint=ParticlePositionWS()
    Position=WorldPosition().XYZ
   )
   Alpha=Clamp(None=Subtract(A=@n4) {constB=0.5})
  )
  n5 = ParticleRelativeTime()
  n6 = LinearInterpolate(A=ScalarParameter() {defaultValue=0.5, parameterName=Color_morning_Hueshift}, B=Constant(), Alpha=Clamp(None=Subtract(A=@n5) {constB=0.2}))
  n7 = ParticleColor()
  n8 = Multiply(A=ParticleRelativeTime())
  n9 = ParticleRelativeTime()
  n10 = LinearInterpolate(A=ScalarParameter() {parameterName=Color_morning_Hueshift_1}, B=Constant(), Alpha=Clamp(None=Subtract(A=@n9) {constB=0.2}))
  n11 = ParticleColor()
  n12 = Multiply(A=ParticleRelativeTime())
  n13 = ParticleRelativeTime()
  n14 = LinearInterpolate(
   A=ScalarParameter() {defaultValue=0.4, parameterName=Color_morning_Hueshift_2}
   B=Constant()
   Alpha=Clamp(None=Subtract(A=@n13) {constB=0.2})
  )
  n15 = ParticleColor()
  n16 = Multiply(A=ParticleRelativeTime())
  n17 = ParticleRelativeTime()
  n18 = LinearInterpolate(
   A=ScalarParameter() {defaultValue=0.1, parameterName=Color_morning_Hueshift_3}
   B=Constant()
   Alpha=Clamp(None=Subtract(A=@n17) {constB=0.2})
  )
  n19 = ParticleColor()
  n20 = Multiply(A=ParticleRelativeTime())
  MaterialFunctionCall(
   Color_Morning=Desaturation(
    None=MaterialFunctionCall(Hue Shift Percentage=@n6, Texture=@n7.RGB) {materialFunction=&HueShift}.Result
    Fraction=Clamp(None=@n8)
   )
   Color_Day=Desaturation(
    None=MaterialFunctionCall(Hue Shift Percentage=@n10, Texture=@n11.RGB) {materialFunction=&HueShift}.Result
    Fraction=Clamp(None=@n12)
   )
   Color_Evening=Desaturation(
    None=MaterialFunctionCall(Hue Shift Percentage=@n14, Texture=@n15.RGB) {materialFunction=&HueShift}.Result
    Fraction=Clamp(None=@n16)
   )
   Color_Night=Desaturation(
    None=MaterialFunctionCall(Hue Shift Percentage=@n18, Texture=@n19.RGB) {materialFunction=&HueShift}.Result
    Fraction=Clamp(None=@n20)
   )
  ) {materialFunction=&Color_Time_of_Day}
  ParticleColor()
  ScalarParameter() {defaultValue=1, parameterName=Metallic}
  ScalarParameter() {parameterName=Specular}
  ScalarParameter() {defaultValue=0.8, parameterName=Roughness}
  VectorParameter() {parameterName=Cubemap Color_1, defaultValue=<rgba 1 1 1 0>}
  VectorParameter() {parameterName=Cubemap Color_2}
  VertexColor()
  VertexNormalWS()
 ]
}
